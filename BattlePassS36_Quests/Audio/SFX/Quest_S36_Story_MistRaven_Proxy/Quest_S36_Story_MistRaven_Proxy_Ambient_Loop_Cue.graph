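# Looping ambient sound cue for the Season 36 MistRaven proxy story quest, routed to the quests sound class.
# A single looping wave feeds a Modulator for slight random variation, then a one-input Mixer at the root, with boosted cue volume.
# Spatialization references the shared HiddenStar supply-drop attenuation asset rather than a quest-specific one.
# Overridden concurrency allows one instance per owner and blocks new starts while it plays.
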SoundCue /BattlePassS36_Quests/Audio/SFX/Quest_S36_Story_MistRaven_Proxy/Quest_S36_Story_MistRaven_Proxy_Ambient_Loop_Cue
root=Mixer_0
nodes [
 {id=Mixer_0, children [Modulator_0], properties {inputVolume [1]}}
 {id=Modulator_0, children [WavePlayer_2], properties {pitchMin=0.97}}
 {
  id=WavePlayer_2
  properties {
   bLooping=true
   soundWaveAssetPtr=/BattlePassS36_Quests/Audio/SFX/Quest_S36_Story_MistRaven_Proxy/Waves/Quest_S36_Story_MistRaven_Proxy_Ambient_Loop
  }
 }
]
properties {
 attenuationSettings=/Game/Athena/Sounds/SupplyDrops/Fort_HiddenStar_Ambient_Loop_Att
 bOverrideConcurrency=true
 maxDistance=2500
 soundClassObject=&Fort_Quests
 volumeMultiplier=2
 attenuationOverrides {
  bAttenuateWithLPF=true
  bEnableFocusInterpolation=true
  bEnableListenerFocus=true
  bEnableOcclusion=true
  distanceAlgorithm=Logarithmic
  falloffDistance=750
  focusAttackInterpSpeed=20
  focusAzimuth=20
  focusReleaseInterpSpeed=20
  hPFFrequencyAtMax=200
  lPFFrequencyAtMax=800
  lPFRadiusMax=500
  lPFRadiusMin=100
  nonFocusVolumeAttenuation=0
  occlusionInterpolationTime=0.3
  occlusionLowPassFilterFrequency=3000
  occlusionTraceChannel=ECC_GameTraceChannel11
  occlusionVolumeAttenuation=0.75
  stereoSpread=300
  attenuationShapeExtents {x=500}
 }
 concurrencyOverrides {bLimitToOwner=true, maxCount=1, resolutionRule=PreventNew}
}
