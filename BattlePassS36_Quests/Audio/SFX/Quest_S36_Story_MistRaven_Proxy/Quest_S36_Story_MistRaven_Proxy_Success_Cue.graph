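# Sound cue for the success sound of the Season 36 Mist Raven proxy story quest.
# A random node picks one of two success waves at equal weight and feeds a modulator, then a mixer whose second input is empty.
# It plays boosted in the non-critical in-game UI sound class, using attenuation borrowed from a Hidden Star ambient loop.
# A concurrency override allows one instance per owner and blocks new plays while it sounds.
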SoundCue /BattlePassS36_Quests/Audio/SFX/Quest_S36_Story_MistRaven_Proxy/Quest_S36_Story_MistRaven_Proxy_Success_Cue
root=Mixer_0
nodes [
 {id=Mixer_0, children [Modulator_0,null], properties {inputVolume [1,1]}}
 {id=Modulator_0, children [Random_0]}
 {id=Random_0, children [WavePlayer_5,WavePlayer_6], properties {weights [1,1]}}
 {
  id=WavePlayer_5
  properties {soundWaveAssetPtr=/BattlePassS36_Quests/Audio/SFX/Quest_S36_Story_MistRaven_Proxy/Waves/Quest_S36_Story_MistRaven_Proxy_Success_01}
 }
 {
  id=WavePlayer_6
  properties {soundWaveAssetPtr=/BattlePassS36_Quests/Audio/SFX/Quest_S36_Story_MistRaven_Proxy/Waves/Quest_S36_Story_MistRaven_Proxy_Success_02}
 }
]
properties {
 attenuationSettings=/Game/Athena/Sounds/SupplyDrops/Fort_HiddenStar_Ambient_Loop_Att
 bOverrideConcurrency=true
 maxDistance=2500
 soundClassObject=&Fort_UI_InGame_NonCritical
 volumeMultiplier=3
 concurrencyOverrides {bLimitToOwner=true, maxCount=1, resolutionRule=PreventNew}
}
